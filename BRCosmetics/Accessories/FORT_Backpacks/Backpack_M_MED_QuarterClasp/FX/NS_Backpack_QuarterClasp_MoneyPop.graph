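# Niagara cosmetic effect for the QuarterClasp backpack that pops V-Bucks meshes and brief glows from sockets.
# VBucks bursts three spawn groups, each setting socket, launch and visibility tag across three mesh renderers, with random colors, gravity, curl noise and delayed drag.
# Glow adds two short time-of-day-tinted sprites.
# BlendSpace switches VBucks between world and local motion by owner speed; both collision modules are disabled.
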NiagaraSystem /BRCosmetics/Accessories/FORT_Backpacks/Backpack_M_MED_QuarterClasp/FX/NS_Backpack_QuarterClasp_MoneyPop
dependencies {
 usedDataInterfaces [
  &NiagaraDataInterfaceArrayColor
  &NiagaraDataInterfaceArrayInt32
  &NiagaraDataInterfaceArrayFloat3
  &NiagaraDataInterfaceCurve
  &NiagaraDataInterfacePlatformSet
  &NiagaraDataInterfaceColorCurve
  &NiagaraDataInterfaceSocketReader
 ]
 usedDynamicInputs []
 usedModules [
  &self:SystemSpawnScript.SystemScriptSource.SystemScriptGraph.NiagaraNodeAssignment_0.SetVariables_#1
  &SystemState
  &BlendSpaceHelper
 ]
}
emitters {
 Glow {
  emitterUpdateScript [
   &EmitterState()
   &SpawnBurst_Instantaneous(Spawn Count=2, Spawn Time=0.05, Use Spawn Probability=true, UseLoopCountLimit=false)
  ]
  particleSpawnScript [
   &PartitionParticles()
   &InitializeParticle(
    Color=&FNTimeOfDayColor(
     Night Color=<rgba 1 0.7134 0.4857 0.23>
     Morning Color=<rgba 1 0.7134 0.4857 0.17>
     Day Color=<rgba 1 0.7134 0.4857 0.12>
     Evening Color=<rgba 1 0.7134 0.4857 0.2>
    )
    Lifetime=0.15
    Mass Mode=::NewEnumerator1
    Mesh Scale Mode=::NewEnumerator3
    Mesh Uniform Scale=0.4
    Sprite Size Mode=::NewEnumerator3
    Uniform Sprite Size=46.6161
    UsePositionOffset=false
    Write Material Random=true
   )
   &SocketLocation(
    Orientation Sampling=::NewEnumerator1
    SamplePosOffset=true
    Sampled Position Offset=&SelectVectorFromArray(
     Direct Array Index=$Particles.PartitionParticles.Partition
     Vector Selection Array=DataInterface() {propertyValues {FloatData [<0 0 -7>, <-15 1 12>]}}
    )
    Socket Index=$Particles.PartitionParticles.Partition
    Socket Reader=$System.SocketReader
    Socket Sampling Mode=::NewEnumerator3
    VelocitySpeedLimit=false
   )
   &CameraOffset(Camera Offset Amount=15)
  ]
  particleUpdateScript [
   &ParticleState()
   &ScaleSpriteSize(Uniform Curve Sprite Scale=DataInterface() {propertyValues {Curve {keys [<at 0 -0.0577 RCIM_Cubic>,<at 0.513 1 RCIM_Cubic>]}}})
   &PointForce(Force Strength=0, Module.Use Falloff=true, Module.UseLinearForce=false)
   &SolveForcesAndVelocity(Clamp Velocity=false, Limit Acceleration=false)
   &ScaleColor(
    Linear Color Curve=DataInterface() {propertyValues {
      AlphaCurve {keys [<at 0 0>,<at 0.1324 1>,<at 1 0>]}
      BlueCurve {keys [<at 0 1>,<at 1 1>]}
      GreenCurve {keys [<at 0 1>,<at 1 1>]}
      RedCurve {keys [<at 0 1>,<at 1 1>]}
     }}
    Scale Mode=2
   )
  ]
  rendererClasses [&NiagaraSpriteRendererProperties]
  renderers [{rendererClass=&NiagaraSpriteRendererProperties}]
 }
 VBucks {
  emitterUpdateScript [
   &EmitterState()
   &SpawnBurst_Instantaneous(Spawn Count=13, Spawn Group=1, Use Spawn Probability=true, UseLoopCountLimit=false)
   &SpawnBurst_Instantaneous(Spawn Count=6, Use Spawn Probability=true, UseLoopCountLimit=false) {module=SpawnBurst_Instantaneous001}
   &SpawnBurst_Instantaneous(Spawn Count=5, Spawn Group=2, Spawn Time=0.4, Use Spawn Probability=true, UseLoopCountLimit=false) {module=SpawnBurst_Instantaneous002}
  ]
  particleSpawnScript [
   &InitializeParticle(
    Color=&SelectLinearColorFromArray(
     Array Sampling Mode=::NewEnumerator0
     Direct Array Index=?
     Color Selection Array=DataInterface() {propertyValues {ColorData [<rgba 1.3 1.1574 0.2302 1>, <rgba 1.3 0.4198 0.6398 1>, <rgba 0.413 1.3 0.7308 1>, <rgba 0.9915 0.5214 1.3 1>, <rgba 1.3 0.56 0.237 1>]}}
     Random Seed=0
    )
    Lifetime=&MergeHelper_FloatSwitch(
     0=&RandomRangeFloat(Minimum=0.5, Maximum=1.4)
     1=&RandomRangeFloat(Minimum=0.5, Maximum=1.4)
     2=&RandomRangeFloat(Minimum=0.4, Maximum=1)
     3=?
     Index=&SpawnGroup()
     Num=::NewEnumerator1
    )
    Mass Mode=::NewEnumerator1
    Mesh Scale Mode=::NewEnumerator3
    Mesh Uniform Scale=0.5
    UsePositionOffset=false
    Write Material Random=true
   )
   &SocketLocation(
    Orientation Sampling=::NewEnumerator1
    SamplePosOffset=true
    Sampled Position Offset=&MergeHelper_VectorSwitch(0=<-5 0 -15>, 1=<-10 0 15>, 2=<0 0 10>, 3=?, Index=&SpawnGroup(), Num=::NewEnumerator1)
    Socket Index=&SelectIntFromArray(Direct Array Index=&SpawnGroup(), Int Selection Array=DataInterface() {propertyValues {IntData [0,1,1]}})
    Socket Reader=$System.SocketReader
    Socket Sampling Mode=::NewEnumerator3
    VelocitySpeedLimit=false
   )
   &InitialMeshOrientation(Use Rotation Vector=false)
   &AddVelocity(
    Cone Angle=&MergeHelper_FloatSwitch(0=5, 1=20, 2=?, 3=?, Index=&SpawnGroup(), Num=::NewEnumerator0)
    Cone Axis=<0 0 1>
    Module.UseConeFalloff=true
    Module.UseVelDistribution=true
    Override Seed=false
    Rotation Angle Type:ENiagara_AngleInput=::NewEnumerator1
    Rotation Mode=::NewEnumerator2
    Speed Falloff From Cone Axis=0.5
    Velocity Mode=::NewEnumerator2
    Velocity Speed=&MergeHelper_FloatSwitch(
     0=&RandomRangeFloat(Minimum=170, Maximum=220)
     1=&RandomRangeFloat(Minimum=200, Maximum=270)
     2=&RandomRangeFloat(Minimum=200, Maximum=230)
     3=?
     Index=&SpawnGroup()
     Num=::NewEnumerator1
    )
    Yaw / Pitch / Roll:Vector3f=&SelectVectorFromArray(
     Direct Array Index=&SpawnGroup()
     Vector Selection Array=DataInterface() {propertyValues {FloatData [<0 -0.02 -0.045>, <0 0.08 -0.03>, <0 0.03 -0.08>]}}
    )
   ) {module=AddVelocity001}
   /SetVariables(Particles.VisibilityTag=&SpawnGroup()) {isSetParameters=true}
   &ShapeLocation(MaskSpawnGroup=false, Override Seed=false, Sphere Radius=5, UseOwnerScale=true) {module=ShapeLocation001}
  ]
  particleUpdateScript [
   /SetVariables(Particles.Drag:NiagaraFloat=&FloatFromCurve(
     FloatCurve=DataInterface() {propertyValues {Curve {keys [
         {arriveTangent=7.4937, interpMode=RCIM_Cubic, leaveTangent=7.4937, tangentMode=RCTM_User, time=0.2038, value=0}
         {
          arriveTangent=-3.4692
          arriveTangentWeight=0.0483
          interpMode=RCIM_Cubic
          leaveTangent=-3.4692
          leaveTangentWeight=0.1448
          tangentMode=RCTM_User
          time=0.2049
          value=1
         }
         <at 1 0.22 RCIM_Cubic>
        ]}}}
     Scale Curve=&RandomRangeFloat(Minimum=8, Maximum=15)
    )) {isSetParameters=true}
   &ParticleState()
   &ScaleMeshSize(Scale Factor=&VectorFromFloat(Value=&FloatFromCurve(
      FloatCurve=DataInterface() {propertyValues {Curve {keys [
          {arriveTangent=4.1785, interpMode=RCIM_Cubic, leaveTangent=8.7018, tangentMode=RCTM_Break, time=0, value=0}
          {arriveTangent=0.0224, interpMode=RCIM_Cubic, leaveTangent=0.0224, tangentMode=RCTM_Break, time=0.3, value=1}
         ]}}}
      CurveIndex=$Particles.Age
      Scale Curve=&FloatFromCurve(FloatCurve=DataInterface() {propertyValues {Curve {keys [
           {interpMode=RCIM_Cubic, tangentMode=RCTM_Break, time=0.7176, value=1}
           {arriveTangent=-9.1079, interpMode=RCIM_Cubic, tangentMode=RCTM_Break, time=1, value=0}
          ]}}})
     )))
   &GravityForce(Gravity=&MergeHelper_VectorSwitch(
     0=&RandomRangeVector(Minimum=<0 0 -980>, Maximum=<1 1 -1100>)
     1=&RandomRangeVector(Minimum=<0 0 -980>, Maximum=<1 1 -1100>)
     2=&RandomRangeVector(Minimum=<0 0 -980>, Maximum=<1 1 -1500>)
     3=?
     Index=&SpawnGroup()
     Num=::NewEnumerator1
    ))
   /Niagara/Modules/Update/Forces/CurlNoiseForce(Noise Frequency=20, Noise Strength=500, Pan Noise=false) {module=CurlNoiseForce001}
   &AerodynamicDrag(
    Aerodynamic Drag:NiagaraFloat=$Particles.Drag
    Aerodynamic Rotational Drag:NiagaraFloat=$Particles.Drag
    Rotation Strength:NiagaraFloat=&RandomRangeFloat(Minimum=3, Maximum=5.5)
    Use Linear Drag=true
    Use Rotational Drag=false
   )
   &Collision(
    Age Colliding Particles=false
    Collision Enabled=&IsPlatformSetActive(New Platform Set=DataInterface() {propertyValues {Platforms {qualityLevelMask=-8}}})
    KillOnCollision=false
    Module.Control Rest State Via Surface Normal=false
    Module.EnableMaxCollisionCount=false
    Module.Randomize Collision Normal Vector=false
    Radius Calculation Type=::NewEnumerator1
   ) {enabled=false}
   &Collision(
    Age Colliding Particles=false
    Analytical Collision Plane Position 1:Vector3f=<0 0 -20>
    Analytical Collision Plane Space:ENiagaraCoordinateSpace=::Local
    CPU Collision Type=::NewEnumerator3
    Collision Enabled=&IsPlatformSetActive(New Platform Set=DataInterface() {propertyValues {Platforms {qualityLevelMask=-28}}})
    KillOnCollision=false
    Module.Control Rest State Via Surface Normal=false
    Module.EnableMaxCollisionCount=false
    Module.Randomize Collision Normal Vector=false
    Perform 2 Analytical Collision Queries=false
    Restitution=0
   ) {enabled=false, module=Collision001}
   &PointForce(Force Strength=0, Module.Use Falloff=true, Module.UseLinearForce=false)
   &SolveForcesAndVelocity(Clamp Velocity=false, Limit Acceleration=false)
   &AlignSpriteToMeshOrientation(SetAlignment=true, SetFacing=true)
   &BlendSpace("Position Blend (World <-> Local)":NiagaraFloat=&MakeCustomFloatFromBool(
     Bool=&SetBoolByFloatComparison(A=&VectorLength(Vector=$Engine.Owner.Velocity), B=300)
     True Float=1
     False Float=0
     Translation Contribution:NiagaraFloat=1
     Rotation Contribution:NiagaraFloat=1
    ))
  ]
  rendererClasses [&NiagaraMeshRendererProperties]
  renderers [
   {rendererClass=&NiagaraMeshRendererProperties, rendererIndex=0}
   {rendererClass=&NiagaraMeshRendererProperties, rendererIndex=1}
   {rendererClass=&NiagaraMeshRendererProperties, rendererIndex=2}
  ]
 }
}
summary {
 systemName=NS_Backpack_QuarterClasp_MoneyPop
 emitters [
  {emitterName=VBucks, rendererClasses [&NiagaraMeshRendererProperties]}
  {emitterName=Glow, rendererClasses [&NiagaraSpriteRendererProperties]}
 ]
 userVariables []
}
