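# Physics asset for the QuarterClasp cosmetic backpack, covering a bag cap, cat and dog figures, and several money pieces.
# Capsule body chains branch from base_bag; two money chains and the dog ears hang off the dog head.
# Spines, necks, heads and most chain roots are zero-size bodies, mostly with free angular constraints.
# Ears, tails, cap and money segments have tight swing and twist limits, allowing small secondary motion.
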PhysicsAsset /BRCosmetics/Accessories/FORT_Backpacks/Backpack_M_MED_QuarterClasp/Meshes/M_MED_QuarterClasp_Pack_Physics
bodies {
 bag_cap_b_01 [{shapeType=Capsule, length=10.2563, radius=1.2126, shapeName=bag_cap_b_01_capsule, center=<0.3049 0 0>, rotation=<rot 0 0 90>}]
 bag_cap_b_02 [{shapeType=Capsule, length=8.6564, radius=1.2126, shapeName=bag_cap_b_02_capsule, center=<0.3049 0 0>, rotation=<rot 0 0 90>}]
 bag_cap_b_03 [{shapeType=Capsule, length=6.5461, radius=1.2126, shapeName=bag_cap_b_03_capsule, center=<0.3049 0 0>, rotation=<rot 0 0 90>}]
 base_bag [{shapeType=Capsule, length=0.505, radius=0.505, shapeName=base_bag_capsule, rotation=<rot 0 -90 0>}]
 cat_ear_l_01 [{shapeType=Capsule, length=0.4525, radius=0.6854, shapeName=cat_ear_l_01_capsule, center=<0.4962 0 0>, rotation=<rot -90 0 0>}]
 cat_ear_l_02 [{shapeType=Capsule, length=0.4525, radius=0.6854, shapeName=cat_ear_l_02_capsule, center=<0.4962 0 0>, rotation=<rot -90 0 0>}]
 cat_ear_r_01 [{shapeType=Capsule, length=0.4525, radius=0.6854, shapeName=cat_ear_r_01_capsule, center=<0.4962 0 0>, rotation=<rot -90 0 0>}]
 cat_ear_r_02 [{shapeType=Capsule, length=0.4525, radius=0.6854, shapeName=cat_ear_r_02_capsule, center=<0.4962 0 0>, rotation=<rot -90 0 0>}]
 cat_head [{shapeType=Capsule, length=0, radius=0, shapeName=cat_head_capsule}]
 cat_neck [{shapeType=Capsule, length=0, radius=0, shapeName=cat_neck_capsule}]
 cat_spine_01 [{shapeType=Capsule, length=0, radius=0, shapeName=cat_spine_01_capsule}]
 cat_spine_02 [{shapeType=Capsule, length=0, radius=0, shapeName=cat_spine_02_capsule}]
 cat_spine_03 [{shapeType=Capsule, length=0, radius=0, shapeName=cat_spine_03_capsule}]
 cat_spine_04 [{shapeType=Capsule, length=0, radius=0, shapeName=cat_spine_04_capsule}]
 cat_tail_01 [{shapeType=Capsule, length=0, radius=0, shapeName=cat_tail_01_capsule}]
 cat_tail_02 [{shapeType=Capsule, length=1.3984, radius=1.0297, shapeName=cat_tail_02_capsule, center=<1.4653 0 0>, rotation=<rot -90 0 0>}]
 cat_tail_03 [{shapeType=Capsule, length=1.3984, radius=1.0297, shapeName=cat_tail_03_capsule, center=<1.4653 0 0>, rotation=<rot -90 0 0>}]
 cat_tail_04 [{shapeType=Capsule, length=1.3984, radius=1.0297, shapeName=cat_tail_04_capsule, center=<1.4653 0 0>, rotation=<rot -90 0 0>}]
 cat_tail_05 [{shapeType=Capsule, length=1.3984, radius=1.0297, shapeName=cat_tail_05_capsule, center=<1.4653 0 0>, rotation=<rot -90 0 0>}]
 dog_ear_l_01 [{shapeType=Capsule, length=3.321, radius=1.1337, shapeName=dog_ear_l_01_capsule, center=<0.5863 0 0>, rotation=<rot 0 0 90>}]
 dog_ear_l_02 [{shapeType=Capsule, length=2.8459, radius=1.1337, shapeName=dog_ear_l_02_capsule, center=<0.5863 0 0>, rotation=<rot 0 0 90>}]
 dog_ear_l_03 [{shapeType=Capsule, length=1.7868, radius=1.1337, shapeName=dog_ear_l_03_capsule, center=<0.5863 0 0>, rotation=<rot 0 0 90>}]
 dog_ear_r_01 [{shapeType=Capsule, length=2.3334, radius=1.0732, shapeName=dog_ear_r_01_capsule, rotation=<rot 0 0 -90>}]
 dog_ear_r_011 [{shapeType=Capsule, length=2.3334, radius=1.0732, shapeName=dog_ear_r_011_capsule, rotation=<rot 0 0 -90>}]
 dog_ear_r_012 [{shapeType=Capsule, length=1.3281, radius=1.0732, shapeName=dog_ear_r_012_capsule, rotation=<rot 0 0 -90>}]
 dog_head_01 [{shapeType=Capsule, length=0, radius=0, shapeName=dog_head_01_capsule}]
 dog_neck_01 [{shapeType=Capsule, length=0, radius=0, shapeName=dog_neck_01_capsule}]
 dog_spine_02 [{shapeType=Capsule, length=0, radius=0, shapeName=dog_spine_02_capsule}]
 dog_spine_03 [{shapeType=Capsule, length=0, radius=0, shapeName=dog_spine_03_capsule}]
 dog_spine_04 [{shapeType=Capsule, length=0, radius=0, shapeName=dog_spine_04_capsule}]
 dog_tail_01 [{shapeType=Capsule, length=0, radius=0, shapeName=dog_tail_01_capsule}]
 dog_tail_02 [{shapeType=Capsule, length=0, radius=1.7323, shapeName=dog_tail_02_capsule, center=<1.4252 0 0>, rotation=<rot 90 0 0>}]
 dog_tail_03 [{shapeType=Capsule, length=0, radius=1.7323, shapeName=dog_tail_03_capsule, center=<1.4252 0 0>, rotation=<rot 90 0 0>}]
 dog_tail_04 [{shapeType=Capsule, length=0, radius=1.7323, shapeName=dog_tail_04_capsule, center=<1.4252 0 0>, rotation=<rot 90 0 0>}]
 dog_tail_05 [{shapeType=Capsule, length=0, radius=1.7323, shapeName=dog_tail_05_capsule, center=<1.4252 0 0>, rotation=<rot 90 0 0>}]
 dog_tail_06 [{shapeType=Capsule, length=0, radius=1.7323, shapeName=dog_tail_06_capsule, center=<1.4252 0 0>, rotation=<rot 90 0 0>}]
 dog_tail_07 [{shapeType=Capsule, length=0, radius=1.7323, shapeName=dog_tail_07_capsule, center=<1.4252 0 0>, rotation=<rot 90 0 0>}]
 moneyA_01 [{shapeType=Capsule, length=0, radius=0, shapeName=moneyA_01_capsule}]
 moneyA_02 [{shapeType=Capsule, length=0.8429, radius=1.3577, shapeName=moneyA_02_capsule, center=<1 0 0>, rotation=<rot 90 0 0>}]
 moneyA_03 [{shapeType=Capsule, length=0.8429, radius=1.3577, shapeName=moneyA_03_capsule, center=<1 0 0>, rotation=<rot 90 0 0>}]
 moneyB_01 [{shapeType=Capsule, length=0, radius=0, shapeName=moneyB_01_capsule}]
 moneyB_02 [{shapeType=Capsule, length=1.2056, radius=1.1198, shapeName=moneyB_02_capsule, center=<0.7472 0 0>, rotation=<rot 0 0 90>}]
 moneyB_03 [{shapeType=Capsule, length=1.2056, radius=1.1198, shapeName=moneyB_03_capsule, center=<0.7472 0 0>, rotation=<rot 0 0 90>}]
 moneyB_04 [{shapeType=Capsule, length=1.2056, radius=1.1198, shapeName=moneyB_04_capsule, center=<0.7472 0 0>, rotation=<rot 0 0 90>}]
 moneyC_01 [{shapeType=Capsule, length=0, radius=0, shapeName=moneyC_01_capsule}]
 moneyC_02 [{shapeType=Capsule, length=1.6522, radius=1.2976, shapeName=moneyC_02_capsule, center=<1.4713 0 0>, rotation=<rot -90 0 0>}]
 moneyD_01 [{shapeType=Capsule, length=0, radius=0, shapeName=moneyD_01_capsule}]
 moneyD_02 [{shapeType=Capsule, length=0.9935, radius=1.13, shapeName=moneyD_02_capsule, center=<1.0882 0 0>, rotation=<rot -90 0 0>}]
 moneyD_03 [{shapeType=Capsule, length=0.9935, radius=1.13, shapeName=moneyD_03_capsule, center=<1.0882 0 0>, rotation=<rot -90 0 0>}]
 money_01 [{shapeType=Capsule, length=0, radius=0, shapeName=money_01_capsule}]
 money_02 [{shapeType=Capsule, length=0.3599, radius=1.3875, shapeName=money_02_capsule, center=<1.2145 0 0>, rotation=<rot 90 0 0>}]
 money_03 [{shapeType=Capsule, length=0.3599, radius=1.3875, shapeName=money_03_capsule, center=<1.2145 0 0>, rotation=<rot 90 0 0>}]
}
constraints [
 {bone1Name=cat_tail_03, bone2Name=cat_tail_02, swing1LimitDegrees=15, swing2LimitDegrees=15, twistLimitDegrees=5}
 {bone1Name=cat_tail_04, bone2Name=cat_tail_03, swing1LimitDegrees=15, swing2LimitDegrees=15, twistLimitDegrees=5}
 {bone1Name=cat_tail_05, bone2Name=cat_tail_04, swing1LimitDegrees=15, swing2LimitDegrees=15, twistLimitDegrees=5}
 {bone1Name=cat_tail_02, bone2Name=cat_tail_01, swing1LimitDegrees=15, swing2LimitDegrees=15, twistLimitDegrees=5}
 {bone1Name=dog_tail_02, bone2Name=dog_tail_01, swing1LimitDegrees=10, swing2LimitDegrees=10, twistLimitDegrees=3}
 {bone1Name=dog_tail_03, bone2Name=dog_tail_02, swing1LimitDegrees=10, swing2LimitDegrees=10, twistLimitDegrees=3}
 {bone1Name=dog_tail_04, bone2Name=dog_tail_03, swing1LimitDegrees=10, swing2LimitDegrees=10, twistLimitDegrees=3}
 {bone1Name=dog_tail_05, bone2Name=dog_tail_04, swing1LimitDegrees=10, swing2LimitDegrees=10, twistLimitDegrees=3}
 {bone1Name=dog_tail_06, bone2Name=dog_tail_05, swing1LimitDegrees=10, swing2LimitDegrees=10, twistLimitDegrees=3}
 {bone1Name=dog_tail_07, bone2Name=dog_tail_06, swing1LimitDegrees=10, swing2LimitDegrees=10, twistLimitDegrees=3}
 {
  bone1Name=dog_tail_01
  bone2Name=base_bag
  swing1LimitDegrees=45
  swing1Motion=Free
  swing2LimitDegrees=45
  swing2Motion=Free
  twistLimitDegrees=45
  twistMotion=Free
 }
 {
  bone1Name=cat_tail_01
  bone2Name=base_bag
  swing1LimitDegrees=45
  swing1Motion=Free
  swing2LimitDegrees=45
  swing2Motion=Free
  twistLimitDegrees=45
  twistMotion=Free
 }
 {bone1Name=bag_cap_b_01, bone2Name=base_bag, swing1LimitDegrees=5, swing2LimitDegrees=5, twistLimitDegrees=2}
 {bone1Name=bag_cap_b_02, bone2Name=bag_cap_b_01, swing1LimitDegrees=5, swing2LimitDegrees=5, twistLimitDegrees=2}
 {bone1Name=bag_cap_b_03, bone2Name=bag_cap_b_02, swing1LimitDegrees=5, swing2LimitDegrees=5, twistLimitDegrees=2}
 {
  bone1Name=moneyB_01
  bone2Name=base_bag
  swing1LimitDegrees=45
  swing1Motion=Free
  swing2LimitDegrees=45
  swing2Motion=Free
  twistLimitDegrees=45
  twistMotion=Free
 }
 {bone1Name=moneyB_02, bone2Name=moneyB_01, swing1LimitDegrees=5, swing2LimitDegrees=5, twistLimitDegrees=2}
 {bone1Name=moneyB_03, bone2Name=moneyB_02, swing1LimitDegrees=5, swing2LimitDegrees=5, twistLimitDegrees=2}
 {bone1Name=moneyB_04, bone2Name=moneyB_03, swing1LimitDegrees=5, swing2LimitDegrees=5, twistLimitDegrees=2}
 {
  bone1Name=moneyC_01
  bone2Name=base_bag
  swing1LimitDegrees=45
  swing1Motion=Free
  swing2LimitDegrees=45
  swing2Motion=Free
  twistLimitDegrees=45
  twistMotion=Free
 }
 {bone1Name=moneyC_02, bone2Name=moneyC_01, swing1LimitDegrees=5, swing2LimitDegrees=5, twistLimitDegrees=2}
 {
  bone1Name=moneyD_01
  bone2Name=base_bag
  swing1LimitDegrees=45
  swing1Motion=Free
  swing2LimitDegrees=45
  swing2Motion=Free
  twistLimitDegrees=45
  twistMotion=Free
 }
 {bone1Name=moneyD_02, bone2Name=moneyD_01, swing1LimitDegrees=5, swing2LimitDegrees=5, twistLimitDegrees=2}
 {bone1Name=moneyD_03, bone2Name=moneyD_02, swing1LimitDegrees=5, swing2LimitDegrees=5, twistLimitDegrees=2}
 {
  bone1Name=dog_spine_02
  bone2Name=base_bag
  swing1LimitDegrees=45
  swing1Motion=Free
  swing2LimitDegrees=45
  swing2Motion=Free
  twistLimitDegrees=45
  twistMotion=Free
 }
 {
  bone1Name=dog_spine_03
  bone2Name=dog_spine_02
  swing1LimitDegrees=45
  swing1Motion=Free
  swing2LimitDegrees=45
  swing2Motion=Free
  twistLimitDegrees=45
  twistMotion=Free
 }
 {
  bone1Name=dog_spine_04
  bone2Name=dog_spine_03
  swing1LimitDegrees=45
  swing1Motion=Free
  swing2LimitDegrees=45
  swing2Motion=Free
  twistLimitDegrees=45
  twistMotion=Free
 }
 {
  bone1Name=dog_neck_01
  bone2Name=dog_spine_04
  swing1LimitDegrees=45
  swing1Motion=Free
  swing2LimitDegrees=45
  swing2Motion=Free
  twistLimitDegrees=45
  twistMotion=Free
 }
 {
  bone1Name=dog_head_01
  bone2Name=dog_neck_01
  swing1LimitDegrees=45
  swing1Motion=Free
  swing2LimitDegrees=45
  swing2Motion=Free
  twistLimitDegrees=45
  twistMotion=Free
 }
 {bone1Name=dog_ear_l_01, bone2Name=dog_head_01, swing1LimitDegrees=10, swing2LimitDegrees=10, twistLimitDegrees=2}
 {bone1Name=dog_ear_l_02, bone2Name=dog_ear_l_01, swing1LimitDegrees=10, swing2LimitDegrees=10, twistLimitDegrees=2}
 {bone1Name=dog_ear_l_03, bone2Name=dog_ear_l_02, swing1LimitDegrees=10, swing2LimitDegrees=10, twistLimitDegrees=2}
 {bone1Name=dog_ear_r_01, bone2Name=dog_head_01, swing1LimitDegrees=10, swing2LimitDegrees=10, twistLimitDegrees=2}
 {bone1Name=dog_ear_r_011, bone2Name=dog_ear_r_01, swing1LimitDegrees=10, swing2LimitDegrees=10, twistLimitDegrees=2}
 {bone1Name=dog_ear_r_012, bone2Name=dog_ear_r_011, swing1LimitDegrees=10, swing2LimitDegrees=10, twistLimitDegrees=2}
 {bone1Name=moneyA_01, bone2Name=dog_head_01, swing1LimitDegrees=5, swing2LimitDegrees=5, twistLimitDegrees=2}
 {bone1Name=moneyA_02, bone2Name=moneyA_01, swing1LimitDegrees=15, swing2LimitDegrees=15, twistLimitDegrees=2}
 {bone1Name=moneyA_03, bone2Name=moneyA_02, swing1LimitDegrees=15, swing2LimitDegrees=15, twistLimitDegrees=2}
 {bone1Name=money_01, bone2Name=dog_head_01, swing1LimitDegrees=10, swing2LimitDegrees=10, twistLimitDegrees=2}
 {bone1Name=money_02, bone2Name=money_01, swing1LimitDegrees=15, swing2LimitDegrees=15, twistLimitDegrees=2}
 {bone1Name=money_03, bone2Name=money_02, swing1LimitDegrees=15, swing2LimitDegrees=15, twistLimitDegrees=2}
 {
  bone1Name=cat_spine_01
  bone2Name=base_bag
  swing1LimitDegrees=45
  swing1Motion=Free
  swing2LimitDegrees=45
  swing2Motion=Free
  twistLimitDegrees=45
  twistMotion=Free
 }
 {
  bone1Name=cat_spine_02
  bone2Name=cat_spine_01
  swing1LimitDegrees=45
  swing1Motion=Free
  swing2LimitDegrees=45
  swing2Motion=Free
  twistLimitDegrees=45
  twistMotion=Free
 }
 {
  bone1Name=cat_spine_03
  bone2Name=cat_spine_02
  swing1LimitDegrees=45
  swing1Motion=Free
  swing2LimitDegrees=45
  swing2Motion=Free
  twistLimitDegrees=45
  twistMotion=Free
 }
 {
  bone1Name=cat_spine_04
  bone2Name=cat_spine_03
  swing1LimitDegrees=45
  swing1Motion=Free
  swing2LimitDegrees=45
  swing2Motion=Free
  twistLimitDegrees=45
  twistMotion=Free
 }
 {
  bone1Name=cat_neck
  bone2Name=cat_spine_04
  swing1LimitDegrees=45
  swing1Motion=Free
  swing2LimitDegrees=45
  swing2Motion=Free
  twistLimitDegrees=45
  twistMotion=Free
 }
 {
  bone1Name=cat_head
  bone2Name=cat_neck
  swing1LimitDegrees=45
  swing1Motion=Free
  swing2LimitDegrees=45
  swing2Motion=Free
  twistLimitDegrees=45
  twistMotion=Free
 }
 {bone1Name=cat_ear_l_01, bone2Name=cat_head, swing1LimitDegrees=5, swing2LimitDegrees=5, twistLimitDegrees=2}
 {bone1Name=cat_ear_l_02, bone2Name=cat_ear_l_01, swing1LimitDegrees=5, swing2LimitDegrees=5, twistLimitDegrees=2}
 {bone1Name=cat_ear_r_01, bone2Name=cat_head, swing1LimitDegrees=5, swing2LimitDegrees=5, twistLimitDegrees=2}
 {bone1Name=cat_ear_r_02, bone2Name=cat_ear_r_01, swing1LimitDegrees=5, swing2LimitDegrees=5, twistLimitDegrees=2}
]
