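# Animation Blueprint for the QuarterClasp cosmetic backpack, which plays idle and reactive animations in the front end and in matches.
# A front-end flag blends between a timer-driven idle/reactive state machine, gated by the backpack's part modifier, and one reacting to elimination and chest-open input events.
# Curve-weighted AnimDynamics on the bag handle and a RigidBody pass add secondary motion before a root scale and offset.
# Anim notifies toggle a material parameter to show or hide the money element.
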AnimBlueprint /BRCosmetics/Accessories/FORT_Backpacks/Backpack_M_MED_QuarterClasp/Meshes/M_MED_QuarterClasp_Pack_AnimBP
target_skeleton=&/M_MED_QuarterClasp_Pack_Skeleton
graphs [
 AnimGraph {nodes [
   Input Pose()
   AnimGraphNode_StateMachine_0 = FE Idle + Reactive() {is=title AnimGraphNode_StateMachine}
   AnimGraphNode_StateMachine_1 = BackpackStates() {is=title AnimGraphNode_StateMachine}
   n1 = Get bIsInFrontEnd()
   n2 = Anim Dynamics - Bone: dyn_bag_handle!(
    ComponentPose=Local To Component(LocalPose=Blend Poses by bool(BlendPose_0=@AnimGraphNode_StateMachine_0, BlendPose_1=@AnimGraphNode_StateMachine_1, bActiveValue=@n1) {blendType=Linear})
    AlphaCurveName:Name=DynamicsAlpha
    ExternalForce="0,0,0"
   ) -> (Pose) {
    alphaInputType=Curve
    angularDampingOverride=0.999
    angularSpringConstant=50
    bGravityOverrideInSimSpace=true
    bLinearSpring=true
    bOverrideLinearDamping=true
    bUseGravityOverride=true
    linearDampingOverride=1
    linearSpringConstant=10
    simulationSpace=BoneRelative
    alphaBoolBlend {blendOption=Linear}
    boundBone {boneName=dyn_bag_handle}
    componentAppliedLinearAccClamp=<500 500 500>
    componentLinearAccScale=<0.1 0.1 0.1>
    componentLinearVelScale=<0.1 0.1 0.1>
    gravityOverride {z=-100}
    physicsBodyDefinitions [{
      boundBone {boneName=dyn_bag_handle}
      boxExtents=<5 5 5>
      constraintSetup {angularLimitsMax=<0 0 0>, angularLimitsMin=<0 0 0>, angularTarget=<0 0 0>, linearAxesMax=<1 0.5 0.3>, linearAxesMin=<-1 -0.5 -0.3>}
      localJointOffset=<0 0 2>
     }]
    relativeSpaceBone {boneName=base_bag}
    simSpaceSettings {maxAngularAcceleration=10, maxAngularVelocity=2, simSpaceAngularAlpha=1}
   }
   n3 = RigidBody!(ComponentPose=@n2, AlphaCurveName:Name=DynamicsAlpha, ExternalForce) -> (Pose) {
    alphaInputType=Curve
    overridePhysicsAsset=&/M_MED_QuarterClasp_Pack_Physics
    simSpaceSettings {worldAlpha=1}
   }
   n4 = "Transform (Modify) Bone - Bone: root"(ComponentPose=@n3, Translation="0,1,0", Rotation="0,0,0", Scale="1.2,1.2,1.2") {scaleMode=BMM_Replace, translationMode=BMM_Additive, boneToModify {boneName=root}, scale=<1.2 1.2 1.2>, translation {y=1}}
   Output Pose(Result=Component To Local(ComponentPose=@n4)) {layerGroup=None}
  ]}
 FE Idle + Reactive {
  entry=FE_Idle
  path=AnimGraph.AnimGraphNode_StateMachine_0.FE Idle + Reactive
  ids=id
  nodes [
   AnimStateNode_0 = FE_Idle() {is=title AnimStateNode}
   AnimStateNode_1 = FE_Reactive() {is=title AnimStateNode, stateEntered {guid=#1, notifyName=Entered FE_Reactive_Anim}}
   AnimStateTransitionNode_0 = AnimStateTransitionNode()
   AnimStateTransitionNode_1 = AnimStateTransitionNode() {bAutomaticRuleBasedOnSequencePlayerInState=true}
  ]
  links [
   {
    dst=FE_Reactive
    rule=AnimGraph.AnimGraphNode_StateMachine_0.FE Idle + Reactive.AnimStateTransitionNode_0.Transition
    src=FE_Idle
    via=AnimStateTransitionNode_0
   }
   {
    dst=FE_Idle
    rule=AnimGraph.AnimGraphNode_StateMachine_0.FE Idle + Reactive.AnimStateTransitionNode_1.Transition
    src=FE_Reactive
    via=AnimStateTransitionNode_1
   }
  ]
 }
 FE_Idle {
  path=AnimGraph.AnimGraphNode_StateMachine_0.FE Idle + Reactive.AnimStateNode_0.FE_Idle
  nodes [Output Animation Pose(Result=Sequence Player 'QuarterClasp_Idle_PACK'() {sequence=/BRCosmetics/Accessories/FORT_Backpacks/Backpack_M_MED_QuarterClasp/Animations/QuarterClasp_Idle_PACK})]
 }
 FE_Reactive {
  path=AnimGraph.AnimGraphNode_StateMachine_0.FE Idle + Reactive.AnimStateNode_1.FE_Reactive
  nodes [
   n1 = Sequence Player 'QuarterClasp_Elim_PACK'() {bLoopAnimation=false, sequence=/BRCosmetics/Accessories/FORT_Backpacks/Backpack_M_MED_QuarterClasp/Animations/QuarterClasp_Elim_PACK}
   Output Animation Pose(Result=@n1)
  ]
 }
 Transition {
  path=AnimGraph.AnimGraphNode_StateMachine_0.FE Idle + Reactive.AnimStateTransitionNode_0.Transition
  nodes [Result(bCanEnterTransition=Get Play Reactive FE Anim() -> (Play Reactive FE Anim:bool))]
 }
 Transition {path=AnimGraph.AnimGraphNode_StateMachine_0.FE Idle + Reactive.AnimStateTransitionNode_1.Transition, nodes [Result()]}
 BackpackStates {
  entry=Idle
  path=AnimGraph.AnimGraphNode_StateMachine_1.BackpackStates
  ids=id
  nodes [
   AnimStateNode_0 = Idle() {
    stateEntered {guid=#2}
    stateFullyBlended {guid=#3}
    stateLeft {guid=#4}
   }
   AnimStateNode_1 = Elimination() {
    is=title AnimStateNode
    stateEntered {guid=#5}
    stateFullyBlended {guid=#6}
    stateLeft {guid=#7}
   }
   AnimStateNode_5 = OpenChest() {is=title AnimStateNode}
   AnimStateTransitionNode_0 = AnimStateTransitionNode() {
    blendMode=Linear
    crossfadeDuration=0
    transitionEnd {guid=#8}
    transitionInterrupt {guid=#9}
    transitionStart {guid=#10}
   }
   AnimStateTransitionNode_1 = AnimStateTransitionNode() {bAutomaticRuleBasedOnSequencePlayerInState=true}
   AnimStateTransitionNode_2 = AnimStateTransitionNode() {bAllowInertializationForSelfTransitions=true}
   AnimStateTransitionNode_3 = AnimStateTransitionNode() {bAllowInertializationForSelfTransitions=true, bAutomaticRuleBasedOnSequencePlayerInState=true, crossfadeDuration=0.4}
  ]
  links [
   {
    dst=Idle
    rule=AnimGraph.AnimGraphNode_StateMachine_1.BackpackStates.AnimStateTransitionNode_1.Transition
    src=Elimination
    via=AnimStateTransitionNode_1
   }
   {
    dst=Elimination
    rule=AnimGraph.AnimGraphNode_StateMachine_1.BackpackStates.AnimStateTransitionNode_0.Transition
    src=Idle
    via=AnimStateTransitionNode_0
   }
   {
    dst=OpenChest
    rule=AnimGraph.AnimGraphNode_StateMachine_1.BackpackStates.AnimStateTransitionNode_2.Transition
    src=Idle
    via=AnimStateTransitionNode_2
   }
   {
    dst=Idle
    rule=AnimGraph.AnimGraphNode_StateMachine_1.BackpackStates.AnimStateTransitionNode_3.Transition
    src=OpenChest
    via=AnimStateTransitionNode_3
   }
  ]
 }
 Idle {
  path=AnimGraph.AnimGraphNode_StateMachine_1.BackpackStates.AnimStateNode_0.Idle
  nodes [Output Animation Pose(Result=Sequence Player 'QuarterClasp_Idle_PACK'() {sequence=&QuarterClasp_Idle_PACK})]
 }
 Elimination {
  path=AnimGraph.AnimGraphNode_StateMachine_1.BackpackStates.AnimStateNode_1.Elimination
  nodes [
   n1 = Sequence Player 'QuarterClasp_Elim_PACK'() {bLoopAnimation=false, sequence=&QuarterClasp_Elim_PACK}
   Output Animation Pose(Result=@n1)
  ]
 }
 OpenChest {
  path=AnimGraph.AnimGraphNode_StateMachine_1.BackpackStates.AnimStateNode_5.OpenChest
  nodes [
   n1 = Sequence Player 'QuarterClasp_ChestReactive_PACK'() {
    bLoopAnimation=false
    sequence=/BRCosmetics/Accessories/FORT_Backpacks/Backpack_M_MED_QuarterClasp/Animations/QuarterClasp_ChestReactive_PACK
   }
   Output Animation Pose(Result=@n1)
  ]
 }
 Transition {
  path=AnimGraph.AnimGraphNode_StateMachine_1.BackpackStates.AnimStateTransitionNode_0.Transition
  nodes [Result(bCanEnterTransition=Get bEnemyPlayerEliminated() -> (bEnemyPlayerEliminated:bool))]
 }
 Transition {path=AnimGraph.AnimGraphNode_StateMachine_1.BackpackStates.AnimStateTransitionNode_1.Transition, nodes [Result()]}
 Transition {
  path=AnimGraph.AnimGraphNode_StateMachine_1.BackpackStates.AnimStateTransitionNode_2.Transition
  nodes [Result(bCanEnterTransition=Get bChestOpened() -> (bChestOpened:bool))]
 }
 Transition {path=AnimGraph.AnimGraphNode_StateMachine_1.BackpackStates.AnimStateTransitionNode_3.Transition, nodes [Result()]}
 EventGraph {nodes [
   Event OnCharacterCustomizationLoaded()
   > n1 = Cast To FortPlayerPawn(Object=TryGetFortPawn())
   > n2 = Cast To B_Athena_PartModifier_Backpack_QuarterClasp!(
    execute
    Object=GetModifierForPartType!(self:Fort Player Pawn Object Reference=@n1, PartType:EFortCustomPartType Enum=Backpack) -> (ReturnValue:Custom Character Part Modifier Object Reference)
   ) -> (then, CastFailed, AsB Athena Part Modifier Backpack Quarter Clasp:B Athena Part Modifier Backpack Quarter Clasp Object Reference) {is=title K2Node_DynamicCast}
   > n3 = Branch(Condition=Get FE_ReactivityEnabled!(self:B Athena Part Modifier Backpack Quarter Clasp Object Reference=@n2) -> (FE_ReactivityEnabled:bool))
   > Do Once(Reset=@n5.then)
   > n4 = Delay(Duration=4)
   Set Play Reactive FE Anim(execute=[@n8.then, @n4.then], Play Reactive FE Anim:bool=true) -> (Output_Get:bool)
   n5 = Set Play Reactive FE Anim(execute=@n3.else, Play Reactive FE Anim=false)
   n6 = Event BlueprintProcessAnimInputEvent()
   > n7 = Branch(Condition=IsNamed(self=@n6.AnimInputEvent, InEventName=aie.EnemyPlayerEliminated))
   > Set bEnemyPlayerEliminated(bEnemyPlayerEliminated:bool=true) -> (Output_Get:bool)
   > Delay(Duration=0.2)
   > Set bEnemyPlayerEliminated(bEnemyPlayerEliminated=false)
   Branch(execute=@n7.else, Condition=IsNamed(self=@n6.AnimInputEvent, InEventName=aie.ChestOpened))
   > Set bChestOpened(bChestOpened:bool=true) -> (Output_Get:bool)
   > RetriggerableDelay(Duration=0.2)
   > Set bChestOpened(bChestOpened=false)
   AnimNotify_FE_ReactiveAnim() {is=title K2Node_Event}
   > Set Play Reactive FE Anim(Play Reactive FE Anim=false)
   > n8 = Delay(Duration=10)
   Event BlueprintPostEvaluateAnimation() {is=title K2Node_Event}
   > Set bEnemyPlayerEliminated(bEnemyPlayerEliminated=false)
   > Set Play Reactive FE Anim(Play Reactive FE Anim=false)
   AnimNotify_MoneyAppear() {is=title K2Node_Event}
   > n9 = Get Dynamic Material Instance(SkeletalMeshComponent=GetPartSkeletalMeshComponent())
   > SetScalarParameterValue(self=@n9, ParameterName=Hide Element 01, Value=0)
   AnimNotify_MoneyHide() {is=title K2Node_Event}
   > n10 = Get Dynamic Material Instance(SkeletalMeshComponent=GetPartSkeletalMeshComponent())
   > SetScalarParameterValue(self=@n10, ParameterName=Hide Element 01, Value=1)
   n11 = Branch(Condition=Get bEnemyPlayerEliminated())
   n12 = Get Dynamic Material Instance(SkeletalMeshComponent=GetPartSkeletalMeshComponent())
   SetScalarParameterValue(execute=@n11.then, self=@n12, ParameterName=Hide Element 01, Value=0)
   > Delay(Duration=5.8)
   > SetScalarParameterValue(self=@n12, ParameterName=Hide Element 01, Value=1)
   SetScalarParameterValue(self=@n12, ParameterName=Hide Element 01, Value=0)
   Event BlueprintUpdateAnimation()
   Get Play Reactive FE Anim()
  ]}
]
variables [
 bChestOpened {type=boolean}
 bEnemyPlayerEliminated {type=boolean}
 FrontEnd {type=boolean}
 Play Reactive FE Anim {type=boolean}
]
